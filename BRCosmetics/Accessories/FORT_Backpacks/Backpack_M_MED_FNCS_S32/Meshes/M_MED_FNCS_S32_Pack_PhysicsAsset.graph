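# Physics asset for the M_MED_FNCS_S32_Pack backpack mesh, a back accessory cosmetic for FNCS Season 32.
# It holds a single body on the pack bone, shaped as one angled capsule that approximates the pack's volume.
# It has no constraints, so it defines no jointed or multi-body simulation.
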PhysicsAsset /BRCosmetics/Accessories/FORT_Backpacks/Backpack_M_MED_FNCS_S32/Meshes/M_MED_FNCS_S32_Pack_PhysicsAsset
bodies {pack [{shapeType=Capsule, length=19.2371, radius=17.5277, center=<-1.8085 -2.2908 -8.0105>, rotation=<rot -0.4569 -90 -81.6309>}]}
constraints []
